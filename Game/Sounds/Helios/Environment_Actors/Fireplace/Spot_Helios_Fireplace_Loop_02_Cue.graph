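# Sound cue for a looping fireplace ambience spot sound in the Helios environment, reusing Winterfest fireplace recordings.
# A Mixer layers the looping close fire wave at full level with a much quieter looping far wave.
# An Enveloper at the root applies a rising volume curve to fade the mix in.
# The cue uses a close-cone fireplace attenuation asset and the Athena ambient props sound class.
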SoundCue /Game/Sounds/Helios/Environment_Actors/Fireplace/Spot_Helios_Fireplace_Loop_02_Cue
root=Enveloper_0
nodes [
 {
  id=Enveloper_0
  children [Mixer_0]
  properties {pitchCurve {editorCurveData {keys [{time=0}]}}, volumeCurve {editorCurveData {keys [<at 0 0>,<at 1 1>]}}}
 }
 {id=Mixer_0, children [WavePlayer_0,WavePlayer_1], properties {inputVolume [1,0.2]}}
 {id=WavePlayer_0, properties {bLooping=true, soundWaveAssetPtr=/Game/Athena/Sounds/Events/Winterfest/Winterfest_Fireplace_Close_Loop}}
 {id=WavePlayer_1, properties {bLooping=true, soundWaveAssetPtr=/Game/Athena/Sounds/Events/Winterfest/Winterfest_Fireplace_Far_Loop}}
]
properties {
 attenuationSettings=&/Fort_Fireplace_Athena_Close_Cone_Att
 maxDistance=2350
 soundClassObject=&Fort_Ambient_Props_Athena
 volumeMultiplier=0.351
 attenuationOverrides {
  bAttenuate=false
  stereoSpread=500
  attenuationShapeExtents {x=0}
  customAttenuationCurve {editorCurveData {keys [{interpMode=RCIM_Cubic, tangentMode=RCTM_Break, time=0, value=1}, <at 1 1>]}}
 }
}
