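# Looping fireplace ambience cue for Helios environment actors, played quietly under the indoor ambient sound class.
# Mixes a Winterfest close fireplace loop with a Curie medium fire loop, then spatializes both through a shared close-cone fireplace attenuation asset.
# An outer mixer with an empty second input feeds all quality levels identically.
# A concurrency override caps simultaneous instances at two, with distance-based volume scaling.
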SoundCue /Game/Sounds/Helios/Environment_Actors/Fireplace/Spot_Helios_Fireplace_Loop_03_Cue
root=QualityLevel_1
nodes [
 {
  id=Attenuation_3
  children [Mixer_0]
  properties {
   attenuationSettings=&/Fort_Fireplace_Athena_Close_Cone_Att
   attenuationOverrides {
    distanceAlgorithm=Custom
    falloffDistance=0
    attenuationShapeExtents {x=0}
    customAttenuationCurve {editorCurveData {keys [{interpMode=RCIM_Cubic, tangentMode=RCTM_Break, time=0, value=1}, {interpMode=RCIM_Cubic, tangentMode=RCTM_Break, time=1, value=0}]}}
   }
  }
 }
 {id=Mixer_0, children [WavePlayer_0,WavePlayer_6], properties {inputVolume [1,1]}}
 {id=Mixer_1, children [Attenuation_3, null], properties {inputVolume [1,1]}}
 {id=QualityLevel_1, children [Mixer_1,Mixer_1,Mixer_1,Mixer_1]}
 {id=WavePlayer_0, properties {bLooping=true, soundWaveAssetPtr=/Game/Athena/Sounds/Events/Winterfest/Winterfest_Fireplace_Close_Loop}}
 {id=WavePlayer_6, properties {bLooping=true, soundWaveAssetPtr=/Game/Sounds/Fort_GamePlay_Sounds/Curie_Fires/Curie_Fire_MediumLoop_03}}
]
properties {
 bOverrideConcurrency=true
 maxDistance=2350
 soundClassObject=&Fort_Ambient_Indoors_Athena
 volumeMultiplier=0.125
 concurrencyOverrides {maxCount=2, volumeScaleMode=Distance}
}
